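# UI material for a challenges screen background (S14 PC): a warm radial gradient with a foil backing texture and animated light streaks.
# An aspect-corrected, edge-masked foil texture is tinted and blended toward a second color by a radial gradient, then a rotated elliptical glow is added.
# Three T_LightStreaks channels pan across rotated screen UVs, with per-cell hashes varying scale, position and opacity.
# The streaks are power-shaped into tinted mids and highlights, radially masked, and added on top.
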Material /BRCosmetics/Athena/UI/Challenges/Art/M_UI_S14_PC_Backgrund
graph {
 properties {materialDomain=MD_UI, pixelDepthOffsetMode=PDOM_Legacy, usedShadingModels=Default Lit}
 nodes [
  n1 = MaterialFunctionCall() {materialFunction=&GetUserInterfaceUV}
  n2 = MaterialFunctionCall(Float2=@n1.Pixel Size) {materialFunction=&BreakOutFloat2Components}
  n3 = Max(A=@n2.G)
  n4 = Multiply(
   A=AppendVector(A=Divide(A=@n2.R, B=@n3), B=Constant() {r=2})
   B=ScalarParameter() {defaultValue=0.5, group=Background-A, parameterName=aBackingTexture-Tile}
  )
  n5 = Subtract(A=@n4)
  n6 = Add(
   A=Add(A=Multiply(A=@n5) {constB=-0.5}, B=Multiply(A=@n4, B=@n1.Normalized UV)) {constB=-0.5}
   B=AppendVector(A=ScalarParameter() {group=Transition, parameterName=Transition-Offset-aBacking}, B=Constant())
  ) {constB=0}
  n7 = TextureObjectParameter() {
   group=Background-A
   parameterName=aBackingTexture
   texture=/BRCosmetics/UI/Foundation/Environment/Shell/Materials/Textures/T_UI_HightowerFoil
  }
  n8 = TextureSample(UVs=@n6, Tex=@n7) {samplerSource=SSM_Clamp_WorldGroupSettings}
  n9 = MaterialFunctionCall!(UVs=@n6, Alpha=@n8.A) {materialFunction=/Game/UI/Foundation/Items/ClampAlphaToTransparentBorder}
  n10 = ScalarParameter() {defaultValue=0.85, group=Global, parameterName=Global-BackingTextureOpacity}
  n11 = AppendVector(
   A=ScalarParameter() {defaultValue=-0.5, group=Background-A, parameterName=aMask-X-Offset}
   B=ScalarParameter() {defaultValue=-0.25, group=Background-A, parameterName=aMask-Y-Offset}
  )
  n12 = Multiply(A=Abs(None=Add(A=@n6, B=@n11) {constB=-0.5})) {constB=2}
  n13 = AppendVector(
   A=ScalarParameter() {defaultValue=0.75, group=Background-A, parameterName=aMask-X-Edge-Distance}
   B=ScalarParameter() {defaultValue=0.35, group=Background-A, parameterName=aMask-Y-Edge-Distance}
  )
  n14 = ScalarParameter() {defaultValue=4, group=Background-A, parameterName=aMask-X-Edge-Sharpness}
  n15 = ScalarParameter() {defaultValue=1, group=Background-A, parameterName=aMask-Y-Edge-Sharpness}
  n16 = Saturate(None=Multiply(A=Subtract(A=@n12, B=@n13), B=AppendVector(A=@n14, B=@n15)))
  n17 = OneMinus(None=Add(A=ComponentMask(None=@n16) {g=true}, B=ComponentMask(None=@n16) {r=true}))
  n18 = LinearInterpolate(B=Clamp(None=LinearInterpolate(B=@n8.RGB, Alpha=@n9.Result) {constA=1}), Alpha=Saturate(None=Multiply(A=@n10, B=@n17))) {constA=1}
  n19 = ScalarParameter() {group=Background-A, parameterName=aRadial-X-Offset}
  n20 = ScalarParameter() {group=Background-A, parameterName=aRadial-Y-Offset}
  n21 = MaterialFunctionCall() {materialFunction=&GetUserInterfaceUV}
  n22 = MaterialFunctionCall(Float2=@n21.Pixel Size) {materialFunction=&BreakOutFloat2Components}
  n23 = Divide(A=@n22.R, B=Max(A=@n22.G))
  n24 = AppendVector(A=Multiply(A=@n23) {constB=0.5}, B=Constant() {r=0.5})
  n25 = Constant() {r=1}
  n26 = Multiply(
   A=Multiply(A=AppendVector(A=@n23, B=@n25), B=@n21.Normalized UV)
   B=ScalarParameter() {defaultValue=1, group=Background-A, parameterName=aRadial-Tile}
  )
  n27 = MaterialFunctionCall(
   UVs=@n26
   CenterPosition=Add(A=@n24, B=AppendVector(A=@n19, B=@n20))
   Radius=ScalarParameter() {defaultValue=0.75, group=Background-A, parameterName=aRadial-Radius}
   Density=ScalarParameter() {defaultValue=1.15, group=Background-A, parameterName=aRadial-Density}
  ) {materialFunction=&RadialGradientExponential}
  n28 = ComponentMask(None=Multiply(
    A=LinearInterpolate(A=@n18, Alpha=@n27.RadialGradientExponential)
    B=VectorParameter() {group=Background-A, parameterName=RGB1, defaultValue=<rgba 0.1458 0.0092 0.0068 1>}.RGB
   )) {b=true, g=true, r=true}
  n29 = ComponentMask(None=VectorParameter() {group=Background-A, parameterName=RGB2, defaultValue=<rgba 1 0.0635 0.0312 1>}.RGB) {b=true, g=true, r=true}
  n30 = VectorParameter() {group=Background-A, parameterName=RGB3, defaultValue=<rgba 1 0.3091 0.0981 1>}
  n31 = ScalarParameter() {defaultValue=-0.125, group=Background-A, parameterName=aGlow_Angle}
  n32 = ScalarParameter() {defaultValue=0.05, group=Background-A, parameterName=aGlow_x-Offset}
  n33 = ScalarParameter() {defaultValue=-0.8, group=Background-A, parameterName=aGlow_y-Offset}
  n34 = MaterialFunctionCall(
   UVs=Multiply(A=Rotator(Coordinate=@n26, Time=@n31) {speed=6.2832}, B=Constant2Vector() {g=1, r=1.5})
   CenterPosition=Add(A=@n24, B=AppendVector(A=@n32, B=@n33))
   Radius=ScalarParameter() {defaultValue=0.75, group=Background-A, parameterName=aGlow_Radius}
   Density=ScalarParameter() {defaultValue=1.125, group=Background-A, parameterName=aGlow_Density}
  ) {materialFunction=&RadialGradientExponential}
  n35 = ScalarParameter() {defaultValue=3, group=fx-streaks, parameterName=fx-Power-Hightlights}
  n36 = ScalarParameter() {defaultValue=12, group=fx-streaks, parameterName=rX-Tile}
  n37 = ScalarParameter() {defaultValue=0.5, group=fx-streaks, parameterName=rY-Tile}
  n38 = Rotator(Coordinate=ScreenPosition().ViewportUV, Time=ScalarParameter() {defaultValue=-0.0427, group=fx-streaks, parameterName=fx-angle}) {speed=6.2832}
  n39 = Panner(
   Coordinate=Multiply(A=@n38, B=AppendVector(A=@n36, B=@n37))
   Speed=AppendVector(A=Constant(), B=ScalarParameter() {defaultValue=-0.25, group=fx-streaks, parameterName=rY-Speed})
  )
  n40 = ScalarParameter() {defaultValue=0.75, group=fx-streaks, parameterName=rScale-Offset}
  n41 = Multiply(A=@n40) {constB=-1}
  n42 = MaterialFunctionCall(Vec2=Floor(None=@n39)) {materialFunction=&MF_Hash}
  n43 = ComponentMask(None=@n42.Vec2) {r=true}
  n44 = ScalarParameter() {defaultValue=0.25, group=fx-streaks, parameterName=rLocation-Offset}
  n45 = ComponentMask(None=@n42.Vec2) {r=true}
  n46 = LinearInterpolate(A=Multiply(A=@n44) {constB=-1}, B=@n44, Alpha=@n45)
  n47 = Add(
   A=MaterialFunctionCall(UVs=Frac(None=@n39), Texture Scale=LinearInterpolate(A=@n41, B=@n40, Alpha=@n43)) {materialFunction=&ScaleUVsByCenter}.UVs
   B=AppendVector(A=@n46, B=@n46)
  )
  n48 = ScalarParameter() {defaultValue=1, group=fx-streaks, parameterName=rOpacity-Offset}
  n49 = ScalarParameter() {defaultValue=14, group=fx-streaks, parameterName=gX-Tile}
  n50 = ScalarParameter() {defaultValue=0.75, group=fx-streaks, parameterName=gY-Tile}
  n51 = Panner(
   Coordinate=Multiply(A=@n38, B=AppendVector(A=@n49, B=@n50))
   Speed=AppendVector(A=Constant(), B=ScalarParameter() {defaultValue=-0.2, group=fx-streaks, parameterName=gY-Speed})
  )
  n52 = Frac(None=@n51)
  n53 = MaterialFunctionCall(Float2=Abs(None=Sine(None=@n52))) {materialFunction=&BreakOutFloat2Components}
  n54 = ScalarParameter() {defaultValue=0.5, group=fx-streaks, parameterName=gScale-Offset}
  n55 = Constant2Vector() {g=-1}
  n56 = MaterialFunctionCall(Vec2=Floor(None=Add(A=@n51, B=@n55))) {materialFunction=&MF_Hash}
  n57 = MaterialFunctionCall(UVs=@n52, Texture Scale=LinearInterpolate(A=Multiply(A=@n54) {constB=-1}, B=@n54, Alpha=ComponentMask(None=@n56.Vec2) {r=true})) {materialFunction=&ScaleUVsByCenter}
  n58 = ScalarParameter() {defaultValue=-0.125, group=fx-streaks, parameterName=gLocation-Offset}
  n59 = ComponentMask(None=@n56.Vec2) {g=true}
  n60 = LinearInterpolate(A=Multiply(A=@n58) {constB=-1}, B=@n58, Alpha=@n59)
  n61 = AppendVector(A=@n60, B=@n60)
  n62 = Multiply(
   A=Saturate(None=DotProduct(A=@n53.R, B=@n53.R))
   B=TextureSample(UVs=Add(A=@n57.UVs, B=@n61)) {texture=/BRCosmetics/UI/Foundation/Environment/Shell/Materials/Textures/T_LightStreaks}.G
  )
  n63 = LinearInterpolate(A=ScalarParameter() {defaultValue=4, group=fx-streaks, parameterName=gOpacity-Offset}, Alpha=@n59)
  n64 = ScalarParameter() {defaultValue=6, group=fx-streaks, parameterName=bX-Tile}
  n65 = ScalarParameter() {defaultValue=0.5, group=fx-streaks, parameterName=bY-Tile}
  n66 = Panner(
   Coordinate=Multiply(A=@n38, B=AppendVector(A=@n64, B=@n65))
   Speed=AppendVector(A=Constant(), B=ScalarParameter() {defaultValue=-0.15, group=fx-streaks, parameterName=bY-Speed})
  )
  n67 = Frac(None=@n66)
  n68 = MaterialFunctionCall(Float2=Abs(None=Sine(None=@n67))) {materialFunction=&BreakOutFloat2Components}
  n69 = ScalarParameter() {defaultValue=0.75, group=fx-streaks, parameterName=bScale-Offset}
  n70 = Constant2Vector() {g=-1, r=-1}
  n71 = MaterialFunctionCall(Vec2=Floor(None=Add(A=@n66, B=@n70))) {materialFunction=&MF_Hash}
  n72 = MaterialFunctionCall(UVs=@n67, Texture Scale=LinearInterpolate(A=Multiply(A=@n69) {constB=-1}, B=@n69, Alpha=ComponentMask(None=@n71.Vec2) {r=true})) {materialFunction=&ScaleUVsByCenter}
  n73 = ScalarParameter() {defaultValue=0.125, group=fx-streaks, parameterName=bLocation-Offset}
  n74 = ComponentMask(None=@n71.Vec2) {g=true}
  n75 = LinearInterpolate(A=Multiply(A=@n73) {constB=-1}, B=@n73, Alpha=@n74)
  n76 = AppendVector(A=@n75, B=@n75)
  n77 = Multiply(
   A=Saturate(None=DotProduct(A=@n68.R, B=@n68.R))
   B=TextureSample(UVs=Add(A=@n72.UVs, B=@n76)) {texture=&T_LightStreaks}.B
  )
  n78 = LinearInterpolate(A=ScalarParameter() {defaultValue=4, group=fx-streaks, parameterName=bOpacity-Offset}, Alpha=@n74)
  n79 = Add(
   A=Multiply(
    A=TextureSample(UVs=@n47) {samplerSource=SSM_Clamp_WorldGroupSettings, texture=&T_LightStreaks}.R
    B=LinearInterpolate(A=@n48, Alpha=@n45)
   )
   B=Add(A=Multiply(A=@n62, B=@n63), B=Multiply(A=@n77, B=@n78))
  )
  n80 = VectorParameter() {group=Background-A, parameterName=RGB2, defaultValue=<rgba 1 0.0635 0.0312 1>}
  n81 = Multiply(A=@n80.RGB, B=@n80.RGB)
  n82 = Add(
   A=Multiply(A=Power(Base=@n79, Exp=@n35), B=ScalarParameter() {defaultValue=0.125, group=fx-streaks, parameterName=fx-Hightlight-Multiply})
   B=Power(Base=Multiply(A=@n81, B=@n79), Exp=ScalarParameter() {defaultValue=0.85, group=fx-streaks, parameterName=fx-Power-Mid})
  )
  n83 = AppendVector(
   A=ScalarParameter() {group=fx-streaks, parameterName=fxMask-X-Offset}
   B=ScalarParameter() {defaultValue=-0.8, group=fx-streaks, parameterName=fxMask-Y-Offset}
  )
  n84 = MaterialFunctionCall(Float2=MaterialFunctionCall() {materialFunction=&ScreenResolution}.Visible Resolution) {materialFunction=&BreakOutFloat2Components}
  n85 = Divide(A=@n84.R, B=@n84.R)
  n86 = AppendVector(A=Multiply(A=@n85) {constB=0.5}, B=Constant() {r=0.5})
  n87 = Multiply(
   A=AppendVector(A=@n85, B=Constant() {r=1})
   B=MaterialFunctionCall() {materialFunction=&ScreenAlignedUVs}."X 100%, Y 100%"
  )
  n88 = ScalarParameter() {defaultValue=1, group=fx-streaks, parameterName=fxMask-Tile}
  n89 = LinearInterpolate(
   B=ScalarParameter() {defaultValue=1, group=fx-streaks, parameterName=fx-Transition-Fade}
   Alpha=MaterialFunctionCall(
    UVs=Multiply(A=@n87, B=@n88)
    CenterPosition=Add(A=@n86, B=@n83)
    Radius=ScalarParameter() {defaultValue=0.75, group=fx-streaks, parameterName=fxMask-Radius}
    Density=ScalarParameter() {defaultValue=3, group=fx-streaks, parameterName=fxMask-Density}
   ) {materialFunction=&RadialGradientExponential}.RadialGradientExponential
  )
  Add(
   A=Add(A=LinearInterpolate(A=@n28, B=@n29, Alpha=@n27.RadialGradientExponential), B=Multiply(A=@n30.RGB, B=@n34.RadialGradientExponential))
   B=Multiply(A=Multiply(A=@n82, B=@n89), B=ScalarParameter() {defaultValue=0.5, group=fx-streaks, parameterName=FX-Intensity})
  )
 ]
}
