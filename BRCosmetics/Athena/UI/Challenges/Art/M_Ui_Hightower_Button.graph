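# Translucent UI material for a Hightower challenges button, drawing a procedural angled-edge shape with cut top-left and bottom-right corners.
# Rotated UV gradients, with Time driven by angle parameters, are offset, sharpened and saturated into edge masks, combined through scalar DotProducts as multiplies.
# Four edges and two inner cuts define the interior, which blends from solid border to the RGBA parameter's alpha.
# Two outer corner cuts then trim the result.
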Material /BRCosmetics/Athena/UI/Challenges/Art/M_Ui_Hightower_Button
graph {
 properties {blendMode=BLEND_Translucent, materialDomain=MD_UI, pixelDepthOffsetMode=PDOM_Legacy, usedShadingModels=Default Lit}
 nodes [
  n1 = Rotator(Coordinate=TextureCoordinate() {coordinateIndex=4}, Time=ScalarParameter() {defaultValue=0.005, group=Border, parameterName=Top-Angle}) {speed=6.2832}
  n2 = Multiply(
   A=Subtract(A=ComponentMask(None=@n1) {g=true}, B=ScalarParameter() {defaultValue=0.0971, group=Border, parameterName=Top-RangeOffset, sliderMax=1})
   B=ScalarParameter() {defaultValue=128, group=Feather, parameterName=HorizontalSharpness}
  )
  n3 = Rotator(Coordinate=TextureCoordinate() {coordinateIndex=4}, Time=ScalarParameter() {defaultValue=0.5025, group=Border, parameterName=Bot-Angle}) {speed=6.2832}
  n4 = Multiply(
   A=Subtract(A=ComponentMask(None=@n3) {g=true}, B=ScalarParameter() {defaultValue=0.125, group=Border, parameterName=Bot-RangeOffset, sliderMax=1})
   B=ScalarParameter() {defaultValue=128, group=Feather, parameterName=HorizontalSharpness}
  )
  n5 = Rotator(Coordinate=TextureCoordinate(), Time=ScalarParameter() {defaultValue=0.0025, group=Border, parameterName=Left-Angle}) {speed=6.2832}
  n6 = Multiply(
   A=Subtract(A=ComponentMask(None=@n5) {r=true}, B=ScalarParameter() {defaultValue=0.0952, group=Border, parameterName=Left-RangeOffset, sliderMax=1})
   B=ScalarParameter() {defaultValue=128, group=Feather, parameterName=VerticalSharpness}
  )
  n7 = ComponentMask(None=Rotator(Coordinate=TextureCoordinate(), Time=ScalarParameter() {defaultValue=0.003, group=Border, parameterName=Right-Angle}) {speed=6.2832}) {r=true}
  n8 = ScalarParameter() {defaultValue=0.8952, group=Border, parameterName=Right-RangeOffset, sliderMax=1}
  n9 = OneMinus(None=Multiply(A=Subtract(A=@n7, B=@n8), B=ScalarParameter() {defaultValue=128, group=Feather, parameterName=VerticalSharpness}))
  n10 = DotProduct(A=DotProduct(A=Saturate(None=@n2), B=Saturate(None=@n4)), B=DotProduct(A=Saturate(None=@n6), B=Saturate(None=@n9)))
  n11 = TextureCoordinate()
  n12 = ScalarParameter() {defaultValue=0.8562, group=Cutout, parameterName=InnerTopLeft-Angle}
  n13 = Subtract(
   A=ComponentMask(None=Rotator(Coordinate=@n11, Time=@n12) {speed=6.2832}) {r=true}
   B=ScalarParameter() {defaultValue=0.0369, group=Cutout, parameterName=InnerTopLeft-Offset}
  )
  n14 = ScalarParameter() {defaultValue=64, group=Cutout, parameterName=Cutout_Sharpness}
  n15 = ScalarParameter() {defaultValue=2.3571, group=Cutout, parameterName=InnerBottomRight-Angle}
  n16 = Subtract(
   A=ComponentMask(None=Rotator(Coordinate=@n11, Time=@n15) {speed=6.2832}) {r=true}
   B=ScalarParameter() {defaultValue=0.0369, group=Cutout, parameterName=InnerBottomRight-offset}
  )
  n17 = DotProduct(A=Saturate(None=Multiply(A=@n13, B=@n14) {constB=128}), B=Saturate(None=Multiply(A=@n16, B=@n14)))
  n18 = TextureCoordinate()
  n19 = Rotator(Coordinate=@n18, Time=ScalarParameter() {defaultValue=0.8752, group=Cutout, parameterName=TopLeft-Angle}) {speed=6.2832}
  n20 = ScalarParameter() {defaultValue=64, group=Cutout, parameterName=Cutout_Sharpness}
  n21 = Multiply(
   A=Subtract(A=ComponentMask(None=@n19) {r=true}, B=ScalarParameter() {defaultValue=-0.0202, group=Cutout, parameterName=TopLeft-Offset})
   B=@n20
  ) {constB=128}
  n22 = Rotator(Coordinate=@n18, Time=ScalarParameter() {defaultValue=2.37, group=Cutout, parameterName=BottomRight-Angle}) {speed=6.2832}
  n23 = Multiply(
   A=Subtract(A=ComponentMask(None=@n22) {r=true}, B=ScalarParameter() {defaultValue=-0.025, group=Cutout, parameterName=BottomRight-offset})
   B=@n20
  )
  Multiply(
   A=LinearInterpolate(A=Constant() {r=1}, B=VectorParameter() {parameterName=RGBA, defaultValue=<rgba 1 0 0.0395 0.1>}.A, Alpha=Multiply(A=@n10, B=@n17))
   B=DotProduct(A=Saturate(None=@n21), B=Saturate(None=@n23))
  )
 ]
}
