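# UI-domain translucent material that draws a slanted progress bar, apparently for the spatial Battle Pass challenges UI.
# Progress interpolates the angle and offset between left and right border settings, placing a sharpened rotated-UV edge that blends FillColor against Background_color.
# Four rotated-UV edges with adjustable angle, inset and feather sharpness multiply into a border mask for opacity, with the unfilled area taking the background's alpha.
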Material /BRCosmetics/Athena/UI/Challenges/Art/T_Ui_Spatial_BP_Progress
graph {
 properties {blendMode=BLEND_Translucent, materialDomain=MD_UI, pixelDepthOffsetMode=PDOM_Legacy, usedShadingModels=Default Lit}
 nodes [
  n1 = VectorParameter() {parameterName=Background_color, defaultValue=<rgba 1 1 1 0.5>}
  n2 = ScalarParameter() {defaultValue=-0.002, group=Border, parameterName=Left-Angle}
  n3 = ScalarParameter() {defaultValue=-0.002, group=Border, parameterName=Right-Angle}
  n4 = ScalarParameter() {defaultValue=0.5, parameterName=Progress, sliderMax=1}
  n5 = ComponentMask(None=Rotator(Coordinate=TextureCoordinate() {coordinateIndex=4}, Time=LinearInterpolate(A=@n2, B=@n3, Alpha=@n4)) {speed=6.2832}) {r=true}
  n6 = LinearInterpolate(
   A=ScalarParameter() {defaultValue=0.02, group=Border, parameterName=Left-RangeOffset, sliderMax=1}
   B=ScalarParameter() {defaultValue=0.98, group=Border, parameterName=Right-RangeOffset, sliderMax=1}
   Alpha=ScalarParameter() {defaultValue=0.5, parameterName=Progress, sliderMax=1}
  )
  n7 = Saturate(None=Multiply(A=Subtract(A=@n5, B=@n6), B=ScalarParameter() {defaultValue=512, group=Feather, parameterName=VerticalSharpness}))
  LinearInterpolate(A=VectorParameter() {parameterName=FillColor, defaultValue=<rgba 1 0 0.6785 1>}.RGB, B=@n1.RGB, Alpha=@n7)
  n8 = OneMinus(None=@n7)
  n9 = ComponentMask(None=Rotator(Coordinate=TextureCoordinate() {coordinateIndex=4}, Time=ScalarParameter() {defaultValue=0.002, group=Border, parameterName=Top-Angle}) {speed=6.2832}) {g=true}
  n10 = ScalarParameter() {defaultValue=0.04, group=Border, parameterName=Top-RangeOffset, sliderMax=1}
  n11 = Saturate(None=Multiply(A=Subtract(A=@n9, B=@n10), B=ScalarParameter() {defaultValue=512, group=Feather, parameterName=HorizontalSharpness}))
  n12 = ComponentMask(None=Rotator(Coordinate=TextureCoordinate() {coordinateIndex=4}, Time=ScalarParameter() {defaultValue=0.5025, group=Border, parameterName=Bot-Angle}) {speed=6.2832}) {g=true}
  n13 = ScalarParameter() {defaultValue=0.04, group=Border, parameterName=Bot-RangeOffset, sliderMax=1}
  n14 = Saturate(None=Multiply(A=Subtract(A=@n12, B=@n13), B=ScalarParameter() {group=Feather, parameterName=HorizontalSharpness}))
  n15 = ComponentMask(None=Rotator(Coordinate=TextureCoordinate() {coordinateIndex=4}, Time=ScalarParameter() {defaultValue=-0.002, group=Border, parameterName=Left-Angle}) {speed=6.2832}) {r=true}
  n16 = ScalarParameter() {defaultValue=0.02, group=Border, parameterName=Left-RangeOffset, sliderMax=1}
  n17 = Saturate(None=Multiply(A=Subtract(A=@n15, B=@n16), B=ScalarParameter() {defaultValue=512, group=Feather, parameterName=VerticalSharpness}))
  n18 = TextureCoordinate() {coordinateIndex=4}
  n19 = ScalarParameter() {defaultValue=-0.002, group=Border, parameterName=Right-Angle}
  n20 = Subtract(
   A=ComponentMask(None=Rotator(Coordinate=@n18, Time=@n19) {speed=6.2832}) {r=true}
   B=ScalarParameter() {defaultValue=0.98, group=Border, parameterName=Right-RangeOffset, sliderMax=1}
  )
  n21 = ScalarParameter() {group=Feather, parameterName=VerticalSharpness}
  n22 = Saturate(None=OneMinus(None=Multiply(A=@n20, B=@n21)))
  Multiply(A=Saturate(None=Add(A=@n1.A, B=@n8)), B=DotProduct(A=DotProduct(A=@n11, B=@n14), B=DotProduct(A=@n17, B=@n22)))
 ]
}
